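# Physics asset for the Syko Blue medium male head cosmetic, simulating a dynamic tube attached to the head.
# Capsule bodies cover the head bone and the dyn_tube bone, with the tube capsule offset and rotated along its bone.
# A single constraint joins the tube to the head, allowing asymmetric swing and only slight twist.
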PhysicsAsset /BRCosmetics/Characters/Player/Male/Medium/Heads/M_MED_SykoBlue_Head/Meshes/Male_Medium_Syko_Head_Physics
bodies {
 dyn_tube [{shapeType=Capsule, length=7.25, radius=4.5, center=<4.5895 0 0>, rotation=<rot 90 0 0>}]
 head [{shapeType=Capsule, length=0.75, radius=3.5}]
}
constraints [{bone1Name=dyn_tube, bone2Name=head, swing1LimitDegrees=60, swing2LimitDegrees=30, twistLimitDegrees=5}]
